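# Niagara system for the Season 19 front end that drifts faint, sand-tinted sprites as a wind layer.
# One GPU emitter spawns large sprites in a flattened hemisphere, with user parameters scaling spawn rate and radius.
# Particles drift along -X, bend toward -Y under age-ramped gravity, rotate, and fade and scale in and out.
# Random dynamic material parameters vary each sprite; a rainbow, camera-distance color module is disabled.
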NiagaraSystem /Game/Effects/Fort_Effects/FrontEnd/S19/NS_SandWind_FrontWind
dependencies {
 usedDataInterfaces [&NiagaraDataInterfaceCurve, &NiagaraDataInterfaceColorCurve, &NiagaraDataInterfaceCamera]
 usedDynamicInputs [/Game/Effects/Niagara/DynamicInputs/Math/FN_Fit, &ReturnCameraVector]
 usedModules [
  &self:SystemSpawnScript.SystemScriptSource.SystemScriptGraph.NiagaraNodeAssignment_0.SetVariables_#1
  &SystemState
  &self:SystemSpawnScript.SystemScriptSource.SystemScriptGraph.NiagaraNodeAssignment_1.SetVariables_#2
 ]
}
emitters {Sand {
  simTarget=GPUComputeSim
  emitterUpdateScript [
   &EmitterState()
   &SpawnRate(SpawnRate=&Multiply_Float(A=8, B=$User.SpawnRateScale), Use Spawn Probability=false)
  ]
  particleSpawnScript [
   /SetVariables(Particles.SpriteRotationRate=&RandomRangeFloat(Minimum=-0.35, Maximum=0.35, Evaluation Type=::NewEnumerator1)) {isSetParameters=true}
   &InitializeParticle(
    Color=&RandomRangeLinearColor(Minimum=<rgba 0.8958 0.5718 0.3359 1>, Maximum=<rgba 0.8958 0.5718 0.3359 1>)
    Lifetime Max=4
    Lifetime Min=3
    Lifetime Mode=::NewEnumerator1
    Sprite Rotation Angle=&RandomRangeFloat(Minimum=0, Maximum=0.1)
    Sprite Rotation Mode=::NewEnumerator1
    Sprite Size Mode=::NewEnumerator1
    Uniform Sprite Size Max=200
    Uniform Sprite Size Min=100
    UsePositionOffset=false
    Write Material Random=true
   ) {module=InitializeParticle001}
   &SphereLocation(
    Hemisphere Z=true
    MaskSpawnGroup=false
    Non Uniform Scale=<1 1 0.01>
    Offset=<0 -500 25>
    Override Seed=false
    Sphere Radius=$User.SphereSize
    Surface Only=false
    UseNonUniformScale=true
    UseOffset=true
    UseSphereOrientation=false
   )
   &DynamicMaterialParameters(
    Index 0 Param 1=&RandomRangeFloat(Minimum=0, Maximum=1, Evaluation Type=::NewEnumerator1)
    Index 0 Param 2=&RandomRangeFloat(Minimum=0, Maximum=1, Evaluation Type=::NewEnumerator1)
    Module.Param0WriteEnabled=true
    Module.Param1WriteEnabled=true
    Module.Param2WriteEnabled=true
    Module.Param3WriteEnabled=true
   )
   &CameraOffset(Camera Offset Amount=&RandomRangeFloat(Minimum=0, Maximum=50, Evaluation Type=::NewEnumerator1))
   &AddVelocity(Velocity=&RandomRangeVector(Minimum=<-15 0 0>, Maximum=<-25 0 0>))
  ]
  particleUpdateScript [
   &GravityForce(Gravity=&Multiply_VectorByFloat(
     Vector=<0 -15 0>
     Float=&FloatFromCurve(FloatCurve=DataInterface() {propertyValues {Curve {keys [<at 0 0>,<at 1 1>]}}})
    )) {module=GravityForce001}
   &ParticleState()
   &Drag(Drag=0, Use Linear Drag=true, Use Rotational Drag=false)
   &SolveForcesAndVelocity(Clamp Velocity=false, Limit Acceleration=false)
   &SpriteRotationRate(Delta Time:NiagaraFloat=$Particles.NormalizedAge, Rotation Rate=$Particles.SpriteRotationRate)
   &ScaleColor(
    Scale Alpha=&Multiply_Float(
     A=&FloatFromCurve(
      FloatCurve=DataInterface() {propertyValues {Curve {keys [
          {interpMode=RCIM_Cubic, tangentMode=RCTM_User, time=0, value=0}
          {interpMode=RCIM_Cubic, tangentMode=RCTM_User, time=0.5, value=1}
          {interpMode=RCIM_Cubic, tangentMode=RCTM_User, time=1, value=0}
         ]}}}
      Scale Curve=25
     )
     B=&FloatFromCurve(
      FloatCurve=DataInterface() {propertyValues {Curve {keys [
          {arriveTangent=2.6303, interpMode=RCIM_Cubic, leaveTangent=2.6303, tangentMode=RCTM_User, time=0, value=0}
          {interpMode=RCIM_Cubic, tangentMode=RCTM_User, time=0.25, value=1}
          <at 0.9 1>
          <at 1 0 RCIM_Cubic>
         ]}}}
      Scale Curve=&Multiply_Float(
       A=&NormalizeDistanceRange(
        Start Position:NiagaraPosition=$Particles.Position
        End Position:NiagaraPosition=Linked() {linkedVariable {name=System.CameraPostion, type=Vector3f}}
        Distance:NiagaraFloat=25000
        Invert Normalized Range:NiagaraBool=false
       )
       B=0.25
      )
     )
    )
    Scale RGB=&Multiply_VectorByFloat(
     Vector=&MakeVectorFromLinearColorRGB(Color=&MakeLinearColorFromVectorAndFloat(
       "Vector (RGB)"=&MakeVectorFromLinearColorRGB(Color=&ColorFromCurve(ColorCurve=DataInterface() {propertyValues {
           AlphaCurve {keys [<at 0 1 RCIM_Cubic>,<at 0.9977 1 RCIM_Cubic>]}
           BlueCurve {keys [<at 0.027 0.3372 RCIM_Cubic>,<at 0.2263 0>,<at 0.4153 2>,<at 0.6291 0>,<at 0.8083 1.9736>]}
           GreenCurve {keys [<at 0.027 0 RCIM_Cubic>,<at 0.2263 0.3189>,<at 0.4153 0.312>,<at 0.6291 0.0903>,<at 0.8083 0>]}
           RedCurve {keys [<at 0.027 0.0686 RCIM_Cubic>,<at 0.2263 1>,<at 0.4153 0>,<at 0.6291 0.5>,<at 0.8083 3>]}
          }}))
       "Float (Alpha)"=&FloatFromCurve(FloatCurve=DataInterface() {propertyValues {Curve {keys [
            {interpMode=RCIM_Cubic, tangentMode=RCTM_User, time=0, value=1}
            {interpMode=RCIM_Cubic, tangentMode=RCTM_User, time=0.5, value=1}
            {interpMode=RCIM_Cubic, tangentMode=RCTM_User, time=1, value=1}
           ]}}})
      ))
     Float=25
    )
    ScaleA=true
    ScaleRGB=true
   ) {enabled=false}
   &ScaleColor(
    Scale Alpha=&Multiply_Float(
     A=&FloatFromCurve(
      FloatCurve=DataInterface() {propertyValues {Curve {keys [
          {interpMode=RCIM_Cubic, tangentMode=RCTM_User, time=0, value=0}
          {interpMode=RCIM_Cubic, tangentMode=RCTM_User, time=0.5, value=1}
          {interpMode=RCIM_Cubic, tangentMode=RCTM_User, time=1, value=0}
         ]}}}
      Scale Curve=5
     )
     B=0.01
    )
    Scale RGB=<5 5 5>
    ScaleA=true
    ScaleRGB=true
   ) {module=ScaleColor001}
   &ScaleSpriteSize(Scale Factor=&Vector2DFromFloat(Value=&FloatFromCurve(FloatCurve=DataInterface() {propertyValues {Curve {keys [
          {arriveTangent=2.6303, interpMode=RCIM_Cubic, leaveTangent=2.6303, tangentMode=RCTM_User, time=0, value=0}
          {interpMode=RCIM_Cubic, tangentMode=RCTM_User, time=0.1, value=1}
          <at 0.9 1>
          <at 1 0 RCIM_Cubic>
         ]}}}))) {module=ScaleSpriteSize001}
   &DynamicMaterialParameters(
    Index 0 Param 1=&RandomRangeFloat(Minimum=0, Maximum=1)
    Index 0 Param 2=&RandomRangeFloat(Minimum=0, Maximum=1)
    Index 1 Param 1=&RandomRangeFloat(Minimum=0.02, Maximum=0.05)
    Index 1 Param 2=&RandomRangeFloat(Minimum=0.02, Maximum=0.04)
    Index 1 Param 3=&RandomRangeFloat(Minimum=0, Maximum=0.1)
    Index 1 Param 4=&RandomRangeFloat(Minimum=0, Maximum=0.1)
    Module.Param0WriteEnabled=true
    Module.Param0WriteEnabled1=true
    Module.Param1WriteEnabled=true
    Module.Param1WriteEnabled1=true
    Module.Param2WriteEnabled=true
    Module.Param2WriteEnabled1=true
    Module.Param3WriteEnabled=true
    Module.Param3WriteEnabled1=true
    Write Parameter Index 1=true
   ) {module=DynamicMaterialParameters001}
  ]
  rendererClasses [&NiagaraSpriteRendererProperties]
  renderers [{rendererClass=&NiagaraSpriteRendererProperties}]
 }}
summary {
 systemName=NS_SandWind_FrontWind
 emitters [{emitterName=Sand, simTarget=GPUComputeSim, rendererClasses [&NiagaraSpriteRendererProperties]}]
 userVariables [
  User.Landscape {
   type=NiagaraDataInterfaceLandscape
   description=""
   defaultValue {
    struct=NiagaraExt_VariableValue_DataInterface
    value {
     dataInterface=&self:NiagaraDataInterfaceLandscape_0
     dataInterfaceClass=&NiagaraDataInterfaceLandscape
    }
   }
  }
  User.SpawnRateScale {type=NiagaraFloat, defaultValue=1, description=""}
  User.SphereSize {type=NiagaraFloat, defaultValue=1000, description=""}
 ]
}
